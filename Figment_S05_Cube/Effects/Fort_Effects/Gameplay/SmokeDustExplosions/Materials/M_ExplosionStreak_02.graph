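# Unlit translucent particle sprite material for explosion streaks in smoke and dust effects.
# Particle alpha multiplies the streak texture's RGB to form a faded streak mask.
# A static switch, HUEandINTENSITYparam, selects either plain particle color or a hue-shifted, intensity-scaled version.
# A dynamic parameter supplies BP_HueShift and BP_Intensity, so color and brightness can vary per particle.
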Material /Figment_S05_Cube/Effects/Fort_Effects/Gameplay/SmokeDustExplosions/Materials/M_ExplosionStreak_02
graph {
 properties {
  bAllowNegativeEmissiveColor=true
  bUsedWithParticleSprites=true
  blendMode=BLEND_Translucent
  pixelDepthOffsetMode=PDOM_Legacy
  shadingModel=MSM_Unlit
  usedShadingModels=Unlit
 }
 nodes [
  n1 = ParticleColor()
  Multiply(A=@n1.A, B=TextureSample() {texture=/Game/Effects/Fort_Effects/Gameplay/SmokeDustExplosions/Textures/T_Explosion_Streak_02}.RGB)
  n2 = DynamicParameter() {defaultValue=<rgba 1 1 10 1>, paramNames [none,BP_Intensity,BP_HueShift,none]}
  StaticSwitchParameter(
   True=Multiply(
    A=@n2.BP_Intensity
    B=MaterialFunctionCall(Hue Shift Percentage=@n2.BP_HueShift, Texture=@n1.RGB) {materialFunction=&HueShift}.Result
   )
   False=@n1.RGB
  ) {parameterName=HUEandINTENSITYparam}
 ]
}
